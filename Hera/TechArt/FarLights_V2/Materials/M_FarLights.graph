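# Additive unlit material for instanced distant-light cards, drawing flickering radial glows and flares, apparently split by a vertex-color flare mask.
# Per-instance and custom primitive data drive color, tint, saturation, falloff and flare size.
# Opacity combines noise flicker, time of day and depth fade, with a camera-height fade also defined.
# World position offset blends sprite billboarding with view-angle-scaled offsets, pushes toward camera, and can collapse geometry onto the object position.
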Material /Hera/TechArt/FarLights_V2/Materials/M_FarLights
graph {
 properties {
  bUsedWithInstancedStaticMeshes=true
  bUsedWithNanite=true
  blendMode=BLEND_Additive
  pixelDepthOffsetMode=PDOM_Legacy
  shadingModel=MSM_Unlit
  usedShadingModels=Unlit
 }
 nodes [
  n1 = TextureCoordinate()
  n2 = Constant2Vector() {g=0.5, r=0.5}
  n3 = OneMinus(None=Divide(A=Distance(A=@n1, B=@n2)) {constB=0.5})
  n4 = LinearInterpolate(
   A=ComponentMask(None=NamedRerouteUsage()) {a=true, desc=Falloff}
   B=ComponentMask(None=NamedRerouteUsage()) {a=true, desc=Falloff}
   Alpha=NamedRerouteUsage()
  ) {desc=Falloff}
  n5 = ComponentMask(None=NamedRerouteUsage()) {desc=Falloff, g=true}
  n6 = Time()
  n7 = ScalarParameter() {defaultValue=0.2, group=Light Flickering, parameterName=FlickerSpeed, sortPriority=1}
  n8 = TextureSample(UVs=Add(A=PerInstanceRandom(), B=Multiply(A=@n6, B=@n7))) {samplerSource=SSM_Wrap_WorldGroupSettings, samplerType=SAMPLERTYPE_Masks, texture=/Game/Global/Textures/Default/Noise/WindNoise_Mask_1}
  n9 = Multiply(
   A=Power(Base=Saturate(None=@n3), Exp=Multiply(A=@n4, B=@n5))
   B=LinearInterpolate(B=@n8.R, Alpha=ScalarParameter() {defaultValue=0.5, group=Light Flickering, parameterName=FlickerIntensity, sortPriority=0}) {constA=1}
  )
  n10 = NamedRerouteUsage()
  n11 = ScalarParameter() {defaultValue=0.5, group=Time of Day Opacity, parameterName=Morning/Evening, sortPriority=2}
  n12 = ScalarParameter() {group=Time of Day Opacity, parameterName=Day, sortPriority=0}
  n13 = ScalarParameter() {defaultValue=1, group=Time of Day Opacity, parameterName=Night, sortPriority=1}
  DepthFade(
   Opacity=Multiply(
    A=Multiply(A=@n9, B=@n10)
    B=MaterialFunctionCall(Value_Morning=@n11, Value_Day=@n12, Value_Evening=@n11, Value_Night=@n13) {materialFunction=&Scalar_Time_of_Day}.Result
   )
   FadeDistance=ScalarParameter() {defaultValue=250, group=Opacity, parameterName=DepthFadeDistance, sortPriority=0}
  ) {fadeDistanceDefault=250}
  n14 = ComponentMask(None=NamedRerouteUsage()) {desc=Saturation, g=true}
  n15 = ComponentMask(None=NamedRerouteUsage()) {desc=Saturation, r=true}
  n16 = Desaturation(None=NamedRerouteUsage(), Fraction=OneMinus(None=Multiply(A=@n14, B=@n15)))
  n17 = LinearInterpolate(
   A=ComponentMask(None=NamedRerouteUsage()) {b=true, g=true, r=true}
   B=ComponentMask(None=NamedRerouteUsage()) {b=true, g=true, r=true}
   Alpha=NamedRerouteUsage()
  ) {desc=Global Tint}
  EyeAdaptationInverse(LightValueInput=Multiply(A=Multiply(A=@n16, B=@n17), B=NamedRerouteUsage()))
  n18 = NamedRerouteUsage()
  n19 = MaterialFunctionCall!(Center Location, 0-1 UVs, XY scale=Multiply(A=ComponentMask(None=@n18) {r=true}) {constB=500}, OverridePivot, "Normals (Optional)") {materialFunction=&Sprite}
  n20 = Subtract(A=CameraPositionWS(), B=WorldPosition().XYZ)
  n21 = Multiply(A=@n20, B=Constant3Vector() {constant=<rgba 1 1 0 0>})
  n22 = Normalize(VectorInput=@n20)
  n23 = DotProduct(A=@n22, B=Constant3Vector() {constant=<rgba 0 0 1 0>})
  n24 = Multiply(A=Multiply(A=Normalize(VectorInput=@n21), B=OneMinus(None=@n23)), B=ComponentMask(None=NamedRerouteUsage()) {b=true, desc=ZOffset})
  n25 = NamedRerouteUsage()
  n26 = ScalarParameter() {defaultValue=10, group=WPO, parameterName=CameraDirectionOffset}
  LinearInterpolate(
   A=Subtract(A=ObjectPositionWS(), B=WorldPosition().XYZ)
   B=Add(A=LinearInterpolate(A=@n19.World position offset, B=@n24, Alpha=@n25), B=Multiply(A=@n22, B=@n26))
   Alpha=Ceil(None=NamedRerouteUsage())
  )
  n27 = CameraPositionWS()
  n28 = TransformPosition(None=Constant3Vector()) {transformSourceType=TRANSFORMPOSSOURCE_Instance, transformType=TRANSFORMPOSSOURCE_World}
  n29 = Add(
   A=Subtract(A=ComponentMask(None=@n27) {b=true}, B=ComponentMask(None=@n28) {b=true})
   B=LinearInterpolate(
    A=ScalarParameter() {defaultValue=-10000, group=Camera Height Fade Opacity, parameterName=Flare Fade Distance Offset, sortPriority=3}
    B=ScalarParameter() {defaultValue=-5000, group=Camera Height Fade Opacity, parameterName=Fade Distance Offset, sortPriority=1}
    Alpha=NamedRerouteUsage()
   )
  )
  n30 = LinearInterpolate(
   A=ScalarParameter() {defaultValue=10000, group=Camera Height Fade Opacity, parameterName=Flare Fade Distance, sortPriority=2}
   B=ScalarParameter() {defaultValue=5000, group=Camera Height Fade Opacity, parameterName=Fade Distance, sortPriority=0}
   Alpha=NamedRerouteUsage()
  )
  NamedRerouteDeclaration(None=Saturate(None=Divide(A=@n29, B=@n30))) {name=CameraHeightFade, nodeColor=<rgba 1 0.3984 0 1>}
  NamedRerouteDeclaration(None=VertexColor().G) {name=FlareMask, nodeColor=<rgba 0.2656 0 1 1>}
  n31 = VectorParameter() {
   bUseCustomPrimitiveData=true
   parameterName=AdditionalParams
   primitiveDataIndex=8
   channelNames {a="", b="", g=Saturation, r=FlareSize}
   defaultValue=<rgba 1 1 1 1>
  }
  NamedRerouteDeclaration(None=@n31.RGBA) {name=AdditionalParams, nodeColor=<rgba 0 0.4609 1 1>}
  n32 = VectorParameter() {
   bUseCustomPrimitiveData=true
   parameterName=FlareTintFalloff
   primitiveDataIndex=4
   channelNames {a=Falloff, b=TintB, g=TintG, r=TintR}
   defaultValue=<rgba 1 1 1 1>
  }
  NamedRerouteDeclaration(None=@n32.RGBA) {name=FlareTintFalloff, nodeColor=<rgba 0 1 0.6484 1>}
  n33 = VectorParameter() {
   bUseCustomPrimitiveData=true
   parameterName=GlowTintFalloff
   channelNames {a=Falloff, b=TintB, g=TintG, r=TintR}
   defaultValue=<rgba 1 1 1 1>
  }
  NamedRerouteDeclaration(None=@n33.RGBA) {name=GlowTintFalloff, nodeColor=<rgba 1 0 0.6328 1>}
  NamedRerouteDeclaration(None=PerInstanceCustomData3Vector!(DefaultValue) {desc=Per Instance LightColor*LightIntensity, constDefaultValue=<rgba 1 1 1 1>}) {name=LightColorIntensity, nodeColor=<rgba 0 1 0.4141 1>}
  NamedRerouteDeclaration(None=PerInstanceCustomData3Vector() {dataIndex=3, desc=Per Instance Tint, constDefaultValue=<rgba 1 1 1 1>}) {name=PerInstanceTint, nodeColor=<rgba 1 0.0938 0 1>}
  NamedRerouteDeclaration(None=PerInstanceCustomData3Vector() {dataIndex=6, desc=Per Instance Saturation/Falloff/ZOffset, constDefaultValue=<rgba 1 1 1 1>}) {name=PerInstanceAdditionalParams, nodeColor=<rgba 0 0.2266 1 1>}
 ]
}
